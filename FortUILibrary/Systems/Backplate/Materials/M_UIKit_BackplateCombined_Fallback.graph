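# Translucent UI material that draws the fallback version of the combined UIKit backplate, a rounded panel with several selectable styles.
# A Backplate Switch parameter picks both the color and the opacity from constant-driven variants, most of them flat colors declared through named reroutes.
# Backplate D blends two dark tints, and an offset radial gradient builds a Glow Mask.
# The selected opacity multiplies a rounded-corner shape mask driven by per-corner radius and pixel-radius parameters.
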Material /FortUILibrary/Systems/Backplate/Materials/M_UIKit_BackplateCombined_Fallback
graph {
 properties {
  bUsedWithStaticMesh=false
  blendMode=BLEND_Translucent
  materialDomain=MD_UI
  pixelDepthOffsetMode=PDOM_Legacy
  usedShadingModels=Default Lit
 }
 nodes [
  Constant3Vector() {constant=<rgba 0.026 0.026 0.026 1>}
  n1 = ComponentMask(None=MaterialFunctionCall() {materialFunction=&GetUserInterfaceUV}.Normalized UV) {g=true}
  LinearInterpolate(Alpha=Saturate(None=Multiply(A=@n1) {constB=4})) {constA=0.7, constB=0.95}
  n2 = VectorParameter() {
   group=Shape
   parameterName=Corner Radius
   sortPriority=1
   channelNames {a=BottomRight, b=TopRight, g=BottomLeft, r=TopLeft}
   defaultValue=<rgba 16 16 16 16>
  }
  MaterialExpressionConstant_1 = Constant() {r=0.75}
  MaterialExpressionConstant_2 = Constant() {r=0.4}
  MaterialExpressionConstant_3 = Constant() {r=0.6}
  MaterialExpressionNamedRerouteUsage_0 = NamedRerouteUsage()
  MaterialExpressionConstant_0 = Constant() {r=1}
  n3 = Switch!(
   SwitchValue=ScalarParameter() {defaultValue=3, parameterName=Backplate Switch, sliderMax=3}
   Default
   A=@MaterialExpressionConstant_1
   B=@MaterialExpressionConstant_2
   C=@MaterialExpressionConstant_3
   D=@MaterialExpressionNamedRerouteUsage_0
   E=@MaterialExpressionConstant_0
  ) {inputs [
    {
     inputName=A
     input {
      expression=&self:MaterialExpressionConstant_1
      inputName=None
      mask=0
      maskA=0
      maskB=0
      maskG=0
      maskR=0
      outputIndex=0
     }
    }
    {
     inputName=B
     input {
      expression=&self:MaterialExpressionConstant_2
      inputName=None
      mask=0
      maskA=0
      maskB=0
      maskG=0
      maskR=0
      outputIndex=0
     }
    }
    {
     inputName=C
     input {
      expression=&self:MaterialExpressionConstant_3
      inputName=None
      mask=0
      maskA=0
      maskB=0
      maskG=0
      maskR=0
      outputIndex=0
     }
    }
    {
     inputName=D
     input {
      expression=&self:MaterialExpressionNamedRerouteUsage_0
      inputName=None
      mask=0
      maskA=0
      maskB=0
      maskG=0
      maskR=0
      outputIndex=0
     }
    }
    {
     inputName=E
     input {
      expression=&self:MaterialExpressionConstant_0
      inputName=None
      mask=0
      maskA=0
      maskB=0
      maskG=0
      maskR=0
      outputIndex=0
     }
    }
   ]}
  Multiply(
   A=MaterialFunctionCall(
    Corner Radius=AppendVector(A=@n2, B=@n2)
    Radius in Pixels=ScalarParameter() {defaultValue=1, group=Shape, parameterName=Radius in Pixels, sliderMax=1, sortPriority=2}
   ) {materialFunction=&MF_UI_CoreUI_RoundedMask}.Shape Mask
   B=@n3
  )
  NamedRerouteDeclaration(None=Constant() {r=0.8}) {name=Backplate D Opacity, nodeColor=<rgba 0 0 0 1>}
  NamedRerouteDeclaration(None=Constant3Vector() {constant=<rgba 1 1 1 0>}) {name=Backplate A, nodeColor=<rgba 0 0 0 1>}
  NamedRerouteDeclaration(None=Constant3Vector() {constant=<rgba 0.3864 0.3763 0.4793 1>}) {name=Backplate B, nodeColor=<rgba 0 0 0 1>}
  n4 = Constant3Vector() {constant=<rgba 0.0027 0.0027 0.0065 1>}
  NamedRerouteDeclaration(None=@n4) {name=Backplate C, nodeColor=<rgba 0 0 0 1>}
  NamedRerouteDeclaration(None=LinearInterpolate(A=@n4, B=Constant3Vector() {constant=<rgba 0.0563 0.0563 0.1302 1>}, Alpha=NamedRerouteUsage())) {name=Backplate D, nodeColor=<rgba 0 0 0 1>}
  n5 = MaterialFunctionCall() {materialFunction=&GetUserInterfaceUV}
  n6 = Subtract(
   A=MaterialFunctionCall(UV=MaterialFunctionCall(UVs=@n5.Normalized UV) {materialFunction=&ScaleUVsByCenter}.UVs) {materialFunction=&MF_UI_UVScaling}.ScaleX
   B=Constant2Vector() {g=-1, r=-0.4}
  ) {constB=0.5}
  NamedRerouteDeclaration(None=LinearInterpolate(Alpha=MaterialFunctionCall(UV=@n6) {materialFunction=&MF_UI_RadialGradient}.Smooth)) {name=Glow Mask, nodeColor=<rgba 1 0.3047 0 1>}
  MaterialExpressionNamedRerouteUsage_2 = NamedRerouteUsage()
  MaterialExpressionNamedRerouteUsage_6 = NamedRerouteUsage()
  MaterialExpressionNamedRerouteUsage_7 = NamedRerouteUsage()
  MaterialExpressionNamedRerouteUsage_8 = NamedRerouteUsage()
  MaterialExpressionConstant3Vector_0 = Constant3Vector() {constant=<rgba 0.9301 1 0.0103 1>}
  Switch(
   SwitchValue=ScalarParameter() {defaultValue=3, parameterName=Backplate Switch, sliderMax=3}
   A=@MaterialExpressionNamedRerouteUsage_2
   B=@MaterialExpressionNamedRerouteUsage_6
   C=@MaterialExpressionNamedRerouteUsage_7
   D=@MaterialExpressionNamedRerouteUsage_8
   E=@MaterialExpressionConstant3Vector_0
  ) {inputs [
    {
     inputName=A
     input {
      expression=&self:MaterialExpressionNamedRerouteUsage_2
      inputName=None
      mask=0
      maskA=0
      maskB=0
      maskG=0
      maskR=0
      outputIndex=0
     }
    }
    {
     inputName=B
     input {
      expression=&self:MaterialExpressionNamedRerouteUsage_6
      inputName=None
      mask=0
      maskA=0
      maskB=0
      maskG=0
      maskR=0
      outputIndex=0
     }
    }
    {
     inputName=C
     input {
      expression=&self:MaterialExpressionNamedRerouteUsage_7
      inputName=None
      mask=0
      maskA=0
      maskB=0
      maskG=0
      maskR=0
      outputIndex=0
     }
    }
    {
     inputName=D
     input {
      expression=&self:MaterialExpressionNamedRerouteUsage_8
      inputName=None
      mask=0
      maskA=0
      maskB=0
      maskG=0
      maskR=0
      outputIndex=0
     }
    }
    {
     inputName=E
     input {
      expression=&self:MaterialExpressionConstant3Vector_0
      inputName=None
      mask=1
      maskA=0
      maskB=1
      maskG=1
      maskR=1
      outputIndex=0
     }
    }
   ]}
 ]
}
