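# Temporary translucent UI material for an alert icon, drawing a filled circle with an exclamation mark cut out of it.
# An SDF circle is smoothstepped with a softness value divided by the viewport UI scale, keeping the edge consistent across scales.
# A shared scale value also sizes the centered exclamation texture, which is combined with the circle by Max and inverted to form the mask.
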Material /QuestUI/QuestScreen/Materials/M_UI_AlertIcon_TEMP
graph {
 properties {bAutomaticallySetUsageInEditor=false, blendMode=BLEND_Translucent, materialDomain=MD_UI, usedShadingModels=Default Lit}
 nodes [
  Constant() {r=1}
  NamedRerouteDeclaration(None=Constant() {r=0.8}) {name=Scale, nodeColor=<rgba 0 1 0.0458 1>}
  NamedRerouteDeclaration(None=Constant() {r=0.12}) {name=Softness, nodeColor=<rgba 0 1 0.0458 1>}
  n1 = Divide(A=NamedRerouteUsage(), B=Custom() {code="return View.ViewportScaleUI;", description=ViewportScaleFactor, outputType=CMOT_Float1, inputs []}) {constA=0.1}
  n2 = MaterialFunctionCall!(UVs, Radius=NamedRerouteUsage()) {materialFunction=/FortUILibrary/MaterialFunctions/Shapes/SDFs/MF_SDF_Circle}
  n3 = NamedRerouteUsage()
  n4 = TextureSample(UVs=MaterialFunctionCall(Texture Scale=Subtract(A=@n3) {constB=0.02}) {materialFunction=&ScaleUVsByCenter}.UVs) {
   samplerSource=SSM_Clamp_WorldGroupSettings
   samplerType=SAMPLERTYPE_Alpha
   texture=/FortUILibrary/Textures/Icons/T_UI_Icon_UIKit_ExclamationMark
  }
  OneMinus(None=Max(A=SmoothStep(Max=@n1, Value=@n2.SDF), B=@n4.RGB))
 ]
}
