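# Translucent UI material for the cinematic reward image in the quest screen, with support for a locked look.
# It samples the cinematic reward icon texture with center-scaled UVs and exposes color and alpha through named reroutes.
# An appended image input feeds the shared MF_UI_AddLock function, along with constants for lock scale, anchor, circle backing and locked opacity.
# The IsLockedCrossfade scalar parameter, defaulting to 1, drives the function's lock crossfade input.
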Material /QuestUI/QuestScreen/Materials/M_UI_CinematicRewardImage
graph {
 properties {bAutomaticallySetUsageInEditor=false, blendMode=BLEND_Translucent, materialDomain=MD_UI, usedShadingModels=Default Lit}
 nodes [
  MaterialFunctionCall!(
   InputV4=AppendVector(A=NamedRerouteUsage(), B=NamedRerouteUsage())
   IsLockedCrossfade=ScalarParameter() {defaultValue=1, parameterName=IsLockedCrossfade}
   LockScale=Constant() {r=0.13}
   CircleBackingOpacity=NamedRerouteUsage()
   CircleBackingSoftness=Constant() {r=0.1}
   CircleBackingRadius=Constant() {r=1.3}
   ImageLockedOpacity=NamedRerouteUsage()
   Anchor=Constant3Vector() {constant=<rgba 0.95 0.2 1 0>}
  ) {materialFunction=&/MF_UI_AddLock}
  NamedRerouteDeclaration(None=Constant() {r=0.9}) {name=CircleBackingOpacity, nodeColor=<rgba 0 0.8594 1 1>}
  NamedRerouteDeclaration(None=Constant() {r=0.8}) {name=LockedPreviewOpacity, nodeColor=<rgba 0 0.8594 1 1>}
  n1 = TextureSample(UVs=MaterialFunctionCall(Texture Scale=Constant() {r=0.91}) {materialFunction=&ScaleUVsByCenter}.UVs) {texture=/QuestUI/QuestScreen/Textures/T_UI_QuestIcon_CinematicReward_L}
  NamedRerouteDeclaration(None=@n1.A) {name=ImageAlpha, nodeColor=<rgba 1 0.5391 0 1>}
  NamedRerouteDeclaration(None=@n1.RGB) {name=ImageRGB, nodeColor=<rgba 1 0.5391 0 1>}
 ]
}
