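# WBP_QuestPages is a Common UI page selector for the quest rewards panel, showing the current page between previous and next buttons.
# Clicks or RewardScroll input actions step SelectedPage with wrap-around, and SetSelectedPage refreshes both labels and broadcasts SelectedPageChanged.
# SetStyle switches between plain "X of Y" text and a backplated "X/Y" label.
# PreConstruct applies mobile styling and optionally removes the buttons, while gamepad input with AlwaysShowHints makes them transparent.
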WidgetBlueprint /QuestUI/QuestScreen/RewardsPanel/WBP_QuestPages
graphs [
 EventGraph {nodes [
   "On Clicked (NextButton)"!() -> (OutputDelegate, then, Button:Common Button Base Object Reference) {is=title K2Node_ComponentBoundEvent}
   > SelectNextPage()
   "On Clicked (PreviousButton)"!() -> (OutputDelegate, then, Button:Common Button Base Object Reference) {is=title K2Node_ComponentBoundEvent}
   > SelectPreviousPage()
   SelectNextPage() {is=title K2Node_CustomEvent}
   n1 = Get SelectedPage() -> (SelectedPage:int=0)
   n2 = int - int(A=Get PagesAmount() -> (PagesAmount:int=0))
   n3 = Get SelectedPage() -> (SelectedPage=0)
   > SetSelectedPage!(
    execute
    self
    SelectedOption:int=Select(Option 0:int=0, Option 1:int=int + int(A=@n1), Index="integer < integer"(A=@n3, B=@n2)) -> (ReturnValue:int)
   ) -> (then)
   SelectPreviousPage() {is=title K2Node_CustomEvent}
   n4 = Get SelectedPage() -> (SelectedPage=0)
   n5 = Get PagesAmount() -> (PagesAmount=0)
   > SetSelectedPage(SelectedOption=Select(Option 0=int - int(A=@n5), Option 1=int - int(A=@n4), Index="integer > integer"(A=@n4)))
   Event PreConstruct()
   > SetColorAndOpacity(
    self:User Widget Object Reference=[
     Get Block_Icon!(self:WBP UIKit Button Generic Base Object Reference=Get NextButton() -> (NextButton:WBP UIKit Button Generic Object Reference)) -> (Block_Icon:WBP UIKit Block Image Object Reference)
     Get Block_Icon(self=Get PreviousButton() -> (PreviousButton:WBP UIKit Button Generic Object Reference))
    ]
    InColorAndOpacity:Linear Color Structure=(R=0,G=0,B=0,A=1)
   )
   > UpdatePagesFont()
   > SetCompactUIStyle() {is=title K2Node_CallFunction}
   > UpdateButtonsPlatformStyle()
   > n6 = Branch(Condition=Get bShowButtons() -> (bShowButtons:bool))
   > SetMinDesiredWidth!(
    execute
    self:Text Object Reference=Get TextBlock!(self:WBP UIKit Block Text Object Reference=Get Pages_Style1() -> (Pages_Style1:WBP UIKit Block Text Object Reference)) -> (TextBlock:Common Text Object Reference)
    InMinDesiredWidth:float=115
   ) -> (then)
   RemoveFromParent(execute=@n6.else, self=[Get PreviousButton(), Get NextButton()])
   > SetMinDesiredWidth(self=Get TextBlock(self=Get Pages_Style1()), InMinDesiredWidth=0)
   Event Construct()
   n7 = Get CommonInputSubsystem()
   > Bind Event to On Input Method Changed(self=@n7, Delegate=Create Event())
   > OnInputChanged!(execute, self, bNewInputType:ECommonInputType Enum=GetCurrentInputType(self=@n7)) -> (then)
  ]}
 OnInputChanged {nodes [
   n1 = OnInputChanged!() -> (then, bNewInputType:ECommonInputType Enum) {is=title K2Node_FunctionEntry}
   n2 = "Equal (Enum)"(A=@n1, B=Gamepad)
   n3 = Get AlwaysShowHints() -> (AlwaysShowHints:bool)
   > SetRenderOpacity(
    self=[Get PreviousButton(), Get NextButton()]
    InOpacity=Select(Option 0:float=1, Option 1:float=0, Index=AND Boolean(A=@n2, B=@n3)) -> (ReturnValue:float)
   )
  ]}
 SelectedPageChanged {nodes [SelectedPageChanged!() -> (then, SelectedPip:int) {is=title K2Node_FunctionEntry}]}
 SetBackColor {nodes [
   n1 = SetBackColor!() -> (then, Color:double=4) {is=title K2Node_FunctionEntry}
   > n2 = GetDynamicMaterial(self=Get BackPlate() -> (BackPlate:Image Object Reference))
   > SetScalarParameterValue(self=@n2, ParameterName=Backplate Switch, Value=@n1)
  ]}
 SetCompactUIStyle {nodes [
   SetCompactUIStyle() {is=title K2Node_FunctionEntry}
   > Branch(Condition=IsMobileGame())
   n1 = Get NextButton()
   > SetIconSize!(execute, self:WBP UIKit Button Generic Base Object Reference=@n1, Icon:Vector 2D Structure=(X=48,Y=48)) -> (then)
   > SetButtonPadding!(
    execute
    self:WBP UIKit Button Generic Base Object Reference=@n1
    Padding_Left:float=2
    Padding_Top:float=0
    Padding_Right:float=0
    Padding_Bottom:float=0
   ) -> (then)
   > SetPadding!(
    execute
    self:Grid Slot Object Reference=SlotAsGridSlot(Widget=@n1)
    InPadding_Left:float=8
    InPadding_Top:float=0
    InPadding_Right:float=0
    InPadding_Bottom:float=0
   ) -> (then)
   n2 = Get PreviousButton()
   > SetIconSize(self=@n2, Icon=(X=48,Y=48))
   > SetButtonPadding(self=@n2, Padding_Left=0, Padding_Top=0, Padding_Right=2, Padding_Bottom=0)
   > SetPadding(self=SlotAsGridSlot(Widget=@n2), InPadding_Left=0, InPadding_Top=0, InPadding_Right=8, InPadding_Bottom=0)
  ]}
 SetOptionsAmount {nodes [
   n1 = SetOptionsAmount!() -> (then, PagesAmount:int) {is=title K2Node_FunctionEntry}
   > Set PagesAmount(PagesAmount:int=@n1) -> (Output_Get:int=0)
   > UpdatePagesInfo()
  ]}
 SetSelectedPage {nodes [
   n1 = SetSelectedPage!() -> (then, SelectedOption:int) {is=title K2Node_FunctionEntry}
   > Set SelectedPage(SelectedPage:int=@n1) -> (Output_Get:int=0)
   > UpdatePagesInfo()
   > Call Selected Page Changed!(execute:Exec, self:WBP Quest Pages Object Reference, SelectedPip:int=Get SelectedPage() -> (SelectedPage=0)) -> (then:Exec) {is=title K2Node_CallDelegate}
  ]}
 SetStyle {nodes [
   n1 = SetStyle!() -> (then, RegularStyle:bool=~) {is=title K2Node_FunctionEntry}
   > SetActiveWidget(
    self=Get VisibilitySwitcher() -> (VisibilitySwitcher:Common Visibility Switcher Object Reference)
    Widget=Select(
     Option 0:Widget Object Reference=Get Pages_Style2() -> (Pages_Style2:Grid Panel Object Reference)
     Option 1:Widget Object Reference=Get Pages_Style1()
     Index=@n1
    ) -> (ReturnValue:Widget Object Reference)
   )
  ]}
 UpdateButtonsPlatformStyle {nodes [
   UpdateButtonsPlatformStyle() {is=title K2Node_FunctionEntry}
   n1 = Get PreviousButton()
   n2 = Get NextButton()
   > SetVisibility(
    self=[
     Get Block_Background!(self:WBP UIKit Button Generic Base Object Reference=@n1) -> (Block_Background:WBP UIKit Block Background Base Object Reference)
     Get Block_Background(self=@n2)
     Get Block_Icon(self=@n1)
     Get Block_Icon(self=@n2)
    ]
    InVisibility=Select(Option 0:ESlateVisibility Enum=Collapsed, Option 1:ESlateVisibility Enum=SelfHitTestInvisible, Index=IsMobileGame()) -> (ReturnValue:ESlateVisibility Enum)
   )
  ]}
 UpdatePagesFont {nodes [
   UpdatePagesFont() {is=title K2Node_FunctionEntry}
   > SetFont!(
    execute
    self:WBP UIKit Block Text Object Reference=Get Pages_Style1()
    Font:"Slate Font Info Structure (by ref)"=Get PagesFont() -> (PagesFont:Slate Font Info Structure)
   ) -> (then)
   n1 = Get Pages_Style1()
   > Set TextTransformPolicy!(execute, TextTransformPolicy:ETextTransformPolicy Enum=None, self:Text Object Reference=Get TextBlock(self=@n1)) -> (then, Output_Get:ETextTransformPolicy Enum=None)
   > SetFontMaterial!(execute, self:WBP UIKit Block Text Object Reference=@n1, In Material:Material Instance Object Reference) -> (then)
  ]}
 UpdatePagesInfo {nodes [
   UpdatePagesInfo() {is=title K2Node_FunctionEntry}
   n1 = int + int(A=Get SelectedPage() -> (SelectedPage=0))
   n2 = Get PagesAmount() -> (PagesAmount=0)
   > SetText!(
    execute
    self:WBP UIKit Block Text Object Reference=Get Pages_Style1()
    Text:Text=Format Text!(Format="LOCTEXT(\"{CurrentPage} of {PagesAmount}\")", CurrentPage:int=@n1, PagesAmount:int=@n2) -> (Result)
   ) -> (then)
   > "SetText (Text)"(
    self=Get PagesText_Style2() -> (PagesText_Style2:Common Text Object Reference)
    InText=Format Text(Format="LOCTEXT(\"{CurrentPage}/{PagesAmount}\")", CurrentPage=@n1, PagesAmount=@n2)
   )
  ]}
]
variables [
 AlwaysShowHints {type=boolean, default=true}
 bShowButtons {type=boolean, default=true}
 PagesAmount {type=integer, default=7}
 PagesFont {
  type=SlateFontInfo
  default {
   bForceMonospaced=false
   bMaterialIsStencil=false
   fontMaterial=null
   fontObject=&HeadingNow
   letterSpacing=0
   monospacedWidth=1
   size=15.75
   skewAmount=0
   typefaceFontName=75 Medium
   outlineSettings {
    bApplyOutlineToDropShadows=false
    bMiteredCorners=false
    bSeparateFillAlpha=false
    outlineMaterial=null
    outlineSize=0
    outlineColor=<rgba 0 0 0 1>
   }
  }
 }
 SelectedPage {type=integer}
 SelectedPageChanged {}
]
widgets {
 info {inheritedWidgetCount=0, parentClass=&CommonUserWidget, rootWidgetClass=&GridPanel, widgetCount=8}
 widgets [
  GridPanel_0 {widgetClassPath=&GridPanel, widget=GridPanel_0}
  PreviousButton {
   widgetClassPath=&WBP_UIKit_Button_Generic
   bIsVariable=true
   parent=GridPanel_0
   widget=PreviousButton
   properties {
    bIsFocusable=false
    backgroundMaterial=/FortUILibrary/UIKit/Buttons/MaterialInstances/CTAButtons/MI_UIKit_BgRegular
    enableFocusedSelectedState=false
    iconMaterial=/FortUILibrary/Foundation/Buttons/Materials/Arrows/MI_UIKit_Icon_Button_Previous
    minHeight=32
    minWidth=32
    mobileMinHeight=80
    mobileMinWidth=80
    outlineMaterial=/FortUILibrary/UIKit/Buttons/MaterialInstances/CTAButtons/MI_UIKit_ButtonOutline
    priority=100
    promptHorizontalAlignment=HAlign_Center
    useTextBlock=false
    buttonPadding=<ltrb 0 0 0 0>
    iconPadding=<ltrb 4 4 4 4>
    iconSize=<16 16>
    promptNudge=<0 0>
    triggeringInputAction {dataTable=&/QuestInputActionDataTable, rowName=RewardScrollLeft}
   }
   slot {class=GridSlot, properties {horizontalAlignment=HAlign_Center, layer=10, verticalAlignment=VAlign_Center}}
  }
  NextButton {
   widgetClassPath=&WBP_UIKit_Button_Generic
   bIsVariable=true
   parent=GridPanel_0
   widget=NextButton
   properties {
    bIsFocusable=false
    backgroundMaterial=&MI_UIKit_BgRegular
    enableFocusedSelectedState=false
    iconMaterial=/FortUILibrary/Foundation/Buttons/Materials/Arrows/MI_UIKit_Icon_Button_Next
    minHeight=32
    minWidth=32
    mobileMinHeight=80
    mobileMinWidth=80
    outlineMaterial=&MI_UIKit_ButtonOutline
    priority=100
    promptHorizontalAlignment=HAlign_Center
    useTextBlock=false
    buttonPadding=<ltrb 0 0 0 0>
    iconPadding=<ltrb 4 4 4 4>
    iconSize=<16 16>
    promptNudge=<0 0>
    triggeringInputAction {dataTable=&QuestInputActionDataTable, rowName=RewardScrollRight}
   }
   slot {class=GridSlot, properties {column=2, horizontalAlignment=HAlign_Center, layer=10, verticalAlignment=VAlign_Center}}
  }
  VisibilitySwitcher {
   widgetClassPath=&CommonVisibilitySwitcher
   bIsVariable=true
   parent=GridPanel_0
   widget=VisibilitySwitcher
   slot {class=GridSlot, properties {column=1, verticalAlignment=VAlign_Center}}
  }
  Pages_Style1 {
   widgetClassPath=/FortUILibrary/Systems/ModularButton/Blocks/WBP_UIKit_Block_Text
   bIsVariable=true
   parent=VisibilitySwitcher
   widget=Pages_Style1
   slot {class=CommonVisibilitySwitcherSlot}
  }
  Pages_Style2 {
   widgetClassPath=&GridPanel
   bIsVariable=true
   parent=VisibilitySwitcher
   widget=Pages_Style2
   slot {class=CommonVisibilitySwitcherSlot, properties {horizontalAlignment=HAlign_Center, verticalAlignment=VAlign_Center}}
  }
  BackPlate {
   widgetClassPath=&Image
   bIsVariable=true
   parent=Pages_Style2
   widget=BackPlate
   properties {brush {
     bIsDynamicallyLoaded=false
     drawAs=Image
     imageType=FullColor
     mirroring=NoMirror
     resourceName=None
     resourceObject=/QuestUI/QuestScreen/Materials/MI_UIKit_TileOptionsBG
     tiling=NoTile
     imageSize=<32 18>
     margin=<ltrb 0 0 0 0>
     outlineSettings {
      bUseBrushTransparency=false
      roundingType=HalfHeightRadius
      width=0
      color {colorUseRule=UseColor_Specified, specifiedColor=<rgba 0 0 0 0>}
      cornerRadii=<0 0 0 0>
     }
     tintColor {colorUseRule=UseColor_Specified, specifiedColor=<rgba 1 1 1 1>}
     uVRegion {bIsValid=false, max=<0 0>, min=<0 0>}
    }}
   slot {class=GridSlot, properties {verticalAlignment=VAlign_Center}}
  }
  PagesText_Style2 {
   widgetClassPath=&CommonTextBlock
   bIsVariable=true
   parent=Pages_Style2
   widget=PagesText_Style2
   properties {
    text=1/2
    colorAndOpacity {colorUseRule=UseColor_Specified, specifiedColor=<rgba 0.0048 0.004 0.0116 1>}
    font {
     bForceMonospaced=false
     bMaterialIsStencil=false
     fontMaterial=null
     fontObject=&HeadingNow
     letterSpacing=0
     monospacedWidth=1
     size=12.75
     skewAmount=0
     typefaceFontName=76 Bold
     outlineSettings {
      bApplyOutlineToDropShadows=false
      bMiteredCorners=false
      bSeparateFillAlpha=false
      outlineMaterial=null
      outlineSize=0
      outlineColor=<rgba 0 0 0 1>
     }
    }
    shadowOffset=<0 0>
   }
   slot {class=GridSlot, properties {verticalAlignment=VAlign_Center, padding=<ltrb 10 0 10 0>}}
  }
 ]
}
