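# Quest screen reward details widget for story video (cinematic) rewards, with label, title, body and an alert banner, derived from WBP_QuestRewardBase.
# InitializeQuestRewardWidget calls the parent, then casts the reward view model to the cinematic reward VM and stores it.
# On Construct, a failed cast of the owning player to the frontend controller marks the player as in-game.
# The banner shows only in-game or on mobile, warning that playback is unavailable in-game or may use mobile data.
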WidgetBlueprint /QuestUI/QuestScreen/RewardsPanel/WBP_QuestRewardCinematicDetails
graphs [
 EventGraph {nodes [
   Event Construct()
   > Set IsInGame(IsInGame:bool=NOT Boolean(A=IsInFrontend!() -> (IsFrontend:bool))) -> (Output_Get:bool)
   n1 = IsMobileGame()
   n2 = Get IsInGame() -> (IsInGame:bool)
   > SetVisibility(
    self=Get WarningArea() -> (WarningArea:Overlay Object Reference)
    InVisibility=Select(Option 0=Collapsed, Option 1=SelfHitTestInvisible, Index=OR Boolean(A=@n1, B=@n2))
   )
   n3 = IsMobileGame()
   n4 = Get MobileWarning() -> (MobileWarning:Text)
   > SetText(
    self=Get Text_Warning() -> (Text_Warning:Common Rich Text Block Object Reference)
    InText=Select(
     Option 0:"Text (by ref)"=Select(Option 1=@n4, Index=@n3)
     Option 1:"Text (by ref)"=Get InGameWarning() -> (InGameWarning:Text)
     Index=Get IsInGame()
    ) -> (ReturnValue:"Text (by ref)")
   )
  ]}
 InitializeQuestRewardWidget {nodes [
   n1 = InitializeQuestRewardWidget!() -> (
    then
    QuestContentVM:Quest UI Base Content View Model Object Reference
    QuestRewardVM:Quest UI Reward VM Object Reference
    IsQuestScreen:bool
   ) {is=title K2Node_FunctionEntry}
   > Parent: InitializeQuestRewardWidget!(
    execute
    QuestContentVM:Quest UI Base Content View Model Object Reference=@n1
    QuestRewardVM:Quest UI Reward VM Object Reference=@n1.QuestRewardVM
    IsQuestScreen:bool=@n1.IsQuestScreen
   ) -> (then) {is=title K2Node_CallParentFunction}
   > Set QuestUIRewardCinematicVM(QuestUIRewardCinematicVM:Quest UI Reward Cinematic VM Object Reference=Cast To QuestUIRewardCinematicVM!(Object=@n1.QuestRewardVM) -> (AsQuest UI Reward Cinematic VM:Quest UI Reward Cinematic VM Object Reference, bSuccess:bool=~) {is=title K2Node_DynamicCast}) -> (Output_Get:Quest UI Reward Cinematic VM Object Reference)
  ]}
 IsInFrontend {nodes [
   IsInFrontend() {is=title K2Node_FunctionEntry}
   > n1 = Cast To FortPlayerControllerFrontEnd!(execute, Object=GetOwningPlayer()) -> (then, CastFailed, AsFort Player Controller Front End:Fort Player Controller Front End Object Reference) {is=title K2Node_DynamicCast}
   > Return Node!(execute, IsFrontend:bool=true)
   Return Node(execute=@n1.CastFailed, IsFrontend=false)
  ]}
]
variables [
 InGameWarning {type=string, default=Video can't be played while in game.}
 IsInGame {type=boolean}
 MobileWarning {type=string, default="Video will be downloaded to play.  This may use your mobile data."}
]
widgets {
 info {
  inheritedWidgetCount=0
  parentClass=&/WBP_QuestRewardBase.WBP_QuestRewardBase_C
  rootWidgetClass=&VerticalBox
  widgetCount=8
 }
 widgets [
  VerticalBox_1 {widgetClassPath=&VerticalBox, widget=VerticalBox_1}
  Text_Label {
   widgetClassPath=&CommonTextBlock
   bIsVariable=true
   parent=VerticalBox_1
   widget=Text_Label
   properties {
    applyLineHeightToBottomLine=false
    bAutoCollapseWithEmptyText=true
    clipping=OnDemand
    mobileFontSizeMultiplier=1.2
    pixelSnapping=Disabled
    scrollStyle=/FortUILibrary/Styles/Typography/Foundation/TextScrollStyle-NoFade.TextScrollStyle-NoFade_C
    text=STORY VIDEO
    textOverflowPolicy=Ellipsis
    textTransformPolicy=ToUpper
    colorAndOpacity {colorUseRule=UseColor_Specified, specifiedColor=<rgba 1 1 1 0.7>}
    font {
     bForceMonospaced=false
     bMaterialIsStencil=false
     fontMaterial=null
     fontObject=&HeadingNow
     letterSpacing=10
     monospacedWidth=1
     size=15.75
     skewAmount=0
     typefaceFontName=86 Bold Italic
     outlineSettings {
      bApplyOutlineToDropShadows=false
      bMiteredCorners=false
      bSeparateFillAlpha=false
      outlineMaterial=null
      outlineSize=0
      outlineColor=<rgba 0 0 0 1>
     }
    }
    shadowOffset=<0 0>
   }
   slot {class=VerticalBoxSlot, properties {horizontalAlignment=HAlign_Left, verticalAlignment=VAlign_Center, padding=<ltrb 5 0 0 0>}}
  }
  Text_Title {
   widgetClassPath=&CommonTextBlock
   bIsVariable=true
   parent=VerticalBox_1
   widget=Text_Title
   properties {
    autoWrapText=true
    bAutoCollapseWithEmptyText=true
    bIsScrollingEnabled=false
    scrollOrientation=Orient_Vertical
    scrollStyle=&TextScrollStyle-NoFade_C
    text=DAIGO'S FATE
    textOverflowPolicy=Ellipsis
    textTransformPolicy=ToUpper
    wrappingPolicy=AllowPerCharacterWrapping
    font {
     bForceMonospaced=false
     bMaterialIsStencil=false
     fontMaterial=null
     fontObject=&HeadingNow
     letterSpacing=0
     monospacedWidth=1
     size=32.25
     skewAmount=0
     typefaceFontName=HeadingNow-86BoldItalic
     outlineSettings {
      bApplyOutlineToDropShadows=false
      bMiteredCorners=false
      bSeparateFillAlpha=false
      outlineMaterial=null
      outlineSize=0
      outlineColor=<rgba 0 0 0 1>
     }
    }
    shadowOffset=<0 0>
   }
   slot {class=VerticalBoxSlot}
  }
  Text_Body {
   widgetClassPath=&CommonRichTextBlock
   bIsVariable=true
   parent=VerticalBox_1
   widget=Text_Body
   properties {
    autoWrapText=true
    bOverrideDefaultStyle=true
    lineHeightPercentage=1.15
    mobileTextBlockScale=1.2
    text=Witness Daigo's journey through the portal to accept his fate in the spirit realm
    textStyleSet=/Game/Athena/Items/Cosmetics/Metadata/ItemDescriptionTextStyle
    defaultTextStyleOverride {
     overflowPolicy=Clip
     transformPolicy=None
     colorAndOpacity {colorUseRule=UseColor_Specified, specifiedColor=<rgba 1 1 1 0.7>}
     font {
      bForceMonospaced=false
      bMaterialIsStencil=false
      fontMaterial=null
      fontObject=&HeadingNow
      letterSpacing=0
      monospacedWidth=1
      size=15.75
      skewAmount=0
      typefaceFontName=74 Regular
      outlineSettings {
       bApplyOutlineToDropShadows=true
       bMiteredCorners=false
       bSeparateFillAlpha=false
       outlineMaterial=null
       outlineSize=0
       outlineColor=<rgba 0 0 0 1>
      }
     }
     highlightColor {colorUseRule=UseColor_Specified, specifiedColor=<rgba 0 0 0 0>}
     highlightShape {
      bIsDynamicallyLoaded=false
      drawAs=Image
      imageType=NoImage
      mirroring=NoMirror
      resourceName=None
      resourceObject=null
      tiling=NoTile
      imageSize=<32 32>
      margin=<ltrb 0 0 0 0>
      outlineSettings {
       bUseBrushTransparency=false
       roundingType=HalfHeightRadius
       width=0
       color {colorUseRule=UseColor_Specified, specifiedColor=<rgba 0 0 0 0>}
       cornerRadii=<0 0 0 0>
      }
      tintColor {colorUseRule=UseColor_Specified, specifiedColor=<rgba 1 1 1 1>}
      uVRegion {bIsValid=false, max=<0 0>, min=<0 0>}
     }
     selectedBackgroundColor {colorUseRule=UseColor_Foreground, specifiedColor=<rgba 1 0 1 1>}
     shadowColorAndOpacity=<rgba 0 0 0 1>
     shadowOffset=<0 0>
     strikeBrush {
      bIsDynamicallyLoaded=false
      drawAs=Image
      imageType=NoImage
      mirroring=NoMirror
      resourceName=None
      resourceObject=null
      tiling=NoTile
      imageSize=<32 32>
      margin=<ltrb 0 0 0 0>
      outlineSettings {
       bUseBrushTransparency=false
       roundingType=HalfHeightRadius
       width=0
       color {colorUseRule=UseColor_Specified, specifiedColor=<rgba 0 0 0 0>}
       cornerRadii=<0 0 0 0>
      }
      tintColor {colorUseRule=UseColor_Specified, specifiedColor=<rgba 1 1 1 1>}
      uVRegion {bIsValid=false, max=<0 0>, min=<0 0>}
     }
     underlineBrush {
      bIsDynamicallyLoaded=false
      drawAs=Image
      imageType=NoImage
      mirroring=NoMirror
      resourceName=None
      resourceObject=null
      tiling=NoTile
      imageSize=<32 32>
      margin=<ltrb 0 0 0 0>
      outlineSettings {
       bUseBrushTransparency=false
       roundingType=HalfHeightRadius
       width=0
       color {colorUseRule=UseColor_Specified, specifiedColor=<rgba 0 0 0 0>}
       cornerRadii=<0 0 0 0>
      }
      tintColor {colorUseRule=UseColor_Specified, specifiedColor=<rgba 1 1 1 1>}
      uVRegion {bIsValid=false, max=<0 0>, min=<0 0>}
     }
    }
   }
   slot {class=VerticalBoxSlot}
  }
  WarningArea {widgetClassPath=&Overlay, bIsVariable=true, parent=VerticalBox_1, widget=WarningArea, slot {class=VerticalBoxSlot}}
  WBP_UIKit_Backplate {
   widgetClassPath=&WBP_UIKit_Backplate
   bIsVariable=true
   parent=WarningArea
   widget=WBP_UIKit_Backplate
   properties {brightness=Dark}
   slot {class=OverlaySlot, properties {horizontalAlignment=HAlign_Fill, verticalAlignment=VAlign_Fill}}
  }
  Img_AlertIcon {
   widgetClassPath=&Image
   parent=WarningArea
   widget=Img_AlertIcon
   properties {brush {
     bIsDynamicallyLoaded=false
     drawAs=Image
     imageType=FullColor
     mirroring=NoMirror
     resourceName=None
     resourceObject=/QuestUI/QuestScreen/Materials/M_UI_AlertIcon_TEMP
     tiling=NoTile
     imageSize=<32 32>
     margin=<ltrb 0 0 0 0>
     outlineSettings {
      bUseBrushTransparency=false
      roundingType=HalfHeightRadius
      width=0
      color {colorUseRule=UseColor_Specified, specifiedColor=<rgba 0 0 0 0>}
      cornerRadii=<0 0 0 0>
     }
     tintColor {colorUseRule=UseColor_Specified, specifiedColor=<rgba 1 1 1 1>}
     uVRegion {bIsValid=false, max=<0 0>, min=<0 0>}
    }}
   slot {class=OverlaySlot, properties {padding=<ltrb 10 10 10 10>}}
  }
  Text_Warning {
   widgetClassPath=&CommonRichTextBlock
   bIsVariable=true
   parent=WarningArea
   widget=Text_Warning
   properties {
    applyLineHeightToBottomLine=false
    autoWrapText=true
    bOverrideDefaultStyle=true
    lineHeightPercentage=1.15
    mobileTextBlockScale=1.2
    text="Video will be downloaded to play.  This may use your mobile data."
    textStyleSet=&ItemDescriptionTextStyle
    defaultTextStyleOverride {
     overflowPolicy=Clip
     transformPolicy=None
     colorAndOpacity {colorUseRule=UseColor_Specified, specifiedColor=<rgba 1 1 1 0.7>}
     font {
      bForceMonospaced=false
      bMaterialIsStencil=false
      fontMaterial=null
      fontObject=&HeadingNow
      letterSpacing=0
      monospacedWidth=1
      size=15.75
      skewAmount=0
      typefaceFontName=74 Regular
      outlineSettings {
       bApplyOutlineToDropShadows=true
       bMiteredCorners=false
       bSeparateFillAlpha=false
       outlineMaterial=null
       outlineSize=0
       outlineColor=<rgba 0 0 0 1>
      }
     }
     highlightColor {colorUseRule=UseColor_Specified, specifiedColor=<rgba 0 0 0 0>}
     highlightShape {
      bIsDynamicallyLoaded=false
      drawAs=Image
      imageType=NoImage
      mirroring=NoMirror
      resourceName=None
      resourceObject=null
      tiling=NoTile
      imageSize=<32 32>
      margin=<ltrb 0 0 0 0>
      outlineSettings {
       bUseBrushTransparency=false
       roundingType=HalfHeightRadius
       width=0
       color {colorUseRule=UseColor_Specified, specifiedColor=<rgba 0 0 0 0>}
       cornerRadii=<0 0 0 0>
      }
      tintColor {colorUseRule=UseColor_Specified, specifiedColor=<rgba 1 1 1 1>}
      uVRegion {bIsValid=false, max=<0 0>, min=<0 0>}
     }
     selectedBackgroundColor {colorUseRule=UseColor_Foreground, specifiedColor=<rgba 1 0 1 1>}
     shadowColorAndOpacity=<rgba 0 0 0 1>
     shadowOffset=<0 0>
     strikeBrush {
      bIsDynamicallyLoaded=false
      drawAs=Image
      imageType=NoImage
      mirroring=NoMirror
      resourceName=None
      resourceObject=null
      tiling=NoTile
      imageSize=<32 32>
      margin=<ltrb 0 0 0 0>
      outlineSettings {
       bUseBrushTransparency=false
       roundingType=HalfHeightRadius
       width=0
       color {colorUseRule=UseColor_Specified, specifiedColor=<rgba 0 0 0 0>}
       cornerRadii=<0 0 0 0>
      }
      tintColor {colorUseRule=UseColor_Specified, specifiedColor=<rgba 1 1 1 1>}
      uVRegion {bIsValid=false, max=<0 0>, min=<0 0>}
     }
     underlineBrush {
      bIsDynamicallyLoaded=false
      drawAs=Image
      imageType=NoImage
      mirroring=NoMirror
      resourceName=None
      resourceObject=null
      tiling=NoTile
      imageSize=<32 32>
      margin=<ltrb 0 0 0 0>
      outlineSettings {
       bUseBrushTransparency=false
       roundingType=HalfHeightRadius
       width=0
       color {colorUseRule=UseColor_Specified, specifiedColor=<rgba 0 0 0 0>}
       cornerRadii=<0 0 0 0>
      }
      tintColor {colorUseRule=UseColor_Specified, specifiedColor=<rgba 1 1 1 1>}
      uVRegion {bIsValid=false, max=<0 0>, min=<0 0>}
     }
    }
   }
   slot {class=OverlaySlot, properties {horizontalAlignment=HAlign_Fill, verticalAlignment=VAlign_Fill, padding=<ltrb 50 12 12 12>}}
  }
 ]
}
