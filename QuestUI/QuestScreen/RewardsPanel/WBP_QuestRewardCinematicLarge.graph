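# Large quest reward widget that previews a cinematic reward and plays its video on demand.
# Initialization caches the cinematic view model, sets the preview material's lock crossfade, and shows Play only for available content in frontend outside the quest screen.
# Clicking marks it watched, broadcasts a playspace cinematic event, and pushes a VUID-configured video widget as an activatable panel with primary, interruptible priority.
# Play and skip both fire quest analytics.
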WidgetBlueprint /QuestUI/QuestScreen/RewardsPanel/WBP_QuestRewardCinematicLarge
animations [NewAnimation {
  binding_count=0
  length_seconds=5
  object=&self:NewAnimation
  sequence {bindings []}
 }]
graphs [
 EventGraph {nodes [
   "On Clicked (Button_DebugQuestUI)"!() -> (OutputDelegate, then, Button:Common Button Base Object Reference) {is=title K2Node_ComponentBoundEvent}
   > n1 = Sequence()
   > MarkWatched(self:Quest UI Reward Cinematic VM Object Reference=Get QuestUIRewardCinematicVM() -> (QuestUIRewardCinematicVM:Quest UI Reward Cinematic VM Object Reference))
   > n2 = GetEventRouter!(
    execute
    Scope:Object Class Reference=/Script/FortniteGame.FortPlayspace
    ContextActor:Actor Object Reference=GetOwningPlayer()
    None:%=Get GameplayEventRouterSubsystem() -> (ReturnValue:Gameplay Event Router Subsystem Object Reference) {is=title K2Node_GetSubsystem}
   ) -> (then, ReturnValue:Gameplay Event Router Component Object Reference)
   > Broadcast Event!(
    execute
    self:Gameplay Event Router Component Object Reference=@n2
    EventData:Quest Screen Cinematic Event Structure=Make Quest Screen Cinematic Event!(MediaID_5_#1:String=Get VUID() -> (VUID:String)) -> (QuestScreenCinematicEvent:Quest Screen Cinematic Event Structure) {is=title K2Node_MakeStruct}
    EventContext:Object Reference
   ) -> (then, ReturnValue:bool)
   > n3 = Construct Object from Class(Class=/Script/FortniteUI.FortMediaVideoWidgetConfig, self=GetOwningPlayer()) -> (ReturnValue:Fort Media Video Widget Config Object Reference)
   > Set MediaID!(execute, MediaID:String=Get VUID(), self:Epic Media Video Widget Config Object Reference=@n3) -> (then, Output_Get:String)
   > Set PerfControl!(
    execute
    PerfControl:Bitmask Byte="To Byte (Integer)"!(InInt:int=Make Bitmask!(Bitmask:Bitmask Integer=7) -> (ReturnValue:Bitmask Integer) {is=title K2Node_BitmaskLiteral}) -> (ReturnValue:Byte=0)
    self:Epic Media Video Widget Config Object Reference=@n3
   ) -> (then, Output_Get:Bitmask Byte=None)
   > n4 = Create Widget!(
    execute:Exec
    Class:User Widget Class Reference=/Game/UI/Foundation/Movie/B_FortVideoWidgetBase.B_FortVideoWidgetBase_C
    OwningPlayer:Player Controller Object Reference=GetOwningPlayer()
   ) -> (then:Exec, ReturnValue:B Fort Video Widget Base Object Reference)
   > n5 = Set VideoWidget(VideoWidget:B Fort Video Widget Base Object Reference=@n4) -> (Output_Get:B Fort Video Widget Base Object Reference)
   n6 = OnVideoHasBeenSkipped() {is=title K2Node_CustomEvent}
   Bind Event to On Video Has Been Skipped(execute=@n5.then, self:B Fort Video Widget Base Object Reference=@n5, Delegate=@n6) {is=title K2Node_AddDelegate}
   > SetConfig!(execute, self:Epic Media Video Widget Base Object Reference=@n5, InConfig:Epic Media Video Widget Config Object Reference=@n3) -> (then, ReturnValue:bool)
   > n7 = GetBasePlayer!(execute, self:Epic Media Video Widget Base Object Reference=Get VideoWidget() -> (VideoWidget:B Fort Video Widget Base Object Reference)) -> (then, ReturnValue:Epic Base Streaming Video Object Reference)
   > n8 = Cast To FortBaseStreamingVideo!(execute, Object=@n7) -> (then, CastFailed, AsFort Base Streaming Video:Fort Base Streaming Video Object Reference) {is=title K2Node_DynamicCast}
   > n9 = SetPriorityInfo!(
    execute
    self:Epic Base Streaming Video Object Reference=@n8
    InPriorityInfo_bCanBeInterrupted:bool=True
    InPriorityInfo_bCanInterruptOtherAndHold:bool=false
    InPriorityInfo_Type:EEpicMediaPriorityType Enum=Primary
   ) -> (then)
   TryPushWidgetAsActivatablePanel!(
    execute=[@n9.then, @n8.CastFailed]
    PushWidget:User Widget Object Reference=Get VideoWidget()
    Header:EActivatePanelOption Enum=PlatformDefault
    Footer:EActivatePanelOption Enum=PlatformDefault
    ChatWidget:EActivatePanelOption Enum=PlatformDefault
   ) -> (then)
   > ActivateWidget(self=Get VideoWidget())
   FireQuestAnalytics_ElementInteracted!(
    execute=@n1.then_1
    PlayerController:Player Controller Object Reference
    Action:"String (by ref)"=BR_Quests_Button_PlayCinematicReward
    ContentVM:Quest UI Base Content View Model Object Reference=Get QuestVM() -> (QuestVM:Quest UI Base Content View Model Object Reference)
   ) -> (then)
   FireQuestAnalytics_ElementInteracted(execute=@n6.then, Action=BR_Quests_Button_SkipCinematicReward, ContentVM=Get QuestVM())
  ]}
 InitializeQuestRewardWidget {nodes [
   n1 = InitializeQuestRewardWidget!() -> (
    then
    QuestContentVM:Quest UI Base Content View Model Object Reference
    QuestRewardVM:Quest UI Reward VM Object Reference
    IsQuestScreen:bool
   ) {is=title K2Node_FunctionEntry}
   > Parent: InitializeQuestRewardWidget!(
    execute
    QuestContentVM:Quest UI Base Content View Model Object Reference=@n1
    QuestRewardVM:Quest UI Reward VM Object Reference=@n1.QuestRewardVM
    IsQuestScreen:bool=@n1.IsQuestScreen
   ) -> (then) {is=title K2Node_CallParentFunction}
   > Set QuestVM(QuestVM:Quest UI Base Content View Model Object Reference=@n1) -> (Output_Get:Quest UI Base Content View Model Object Reference)
   > Is Valid(InputObject=@n1.QuestRewardVM)
   n2 = Cast To QuestUIRewardCinematicVM!(Object=@n1.QuestRewardVM) -> (AsQuest UI Reward Cinematic VM:Quest UI Reward Cinematic VM Object Reference, bSuccess:bool=~) {is=title K2Node_DynamicCast}
   > Set QuestUIRewardCinematicVM(QuestUIRewardCinematicVM:Quest UI Reward Cinematic VM Object Reference=@n2) -> (Output_Get:Quest UI Reward Cinematic VM Object Reference)
   > Set VUID(VUID:String=Get VUID!(self:Quest UI Reward Cinematic VM Object Reference=@n2) -> (VUID)) -> (Output_Get:String)
   > Set CinematicAvailable(CinematicAvailable:bool=Get bIsRewardContentAvailable!(self:Quest UI Reward VM Object Reference=@n2) -> (bIsRewardContentAvailable:bool)) -> (Output_Get:bool)
   > n3 = GetDynamicMaterial(self=Get PreviewImage() -> (PreviewImage:Image Object Reference))
   > SetScalarParameterValue(
    self=@n3
    ParameterName=IsLockedCrossfade
    Value=Select(Option 0:float=1, Option 1:float=0, Index=Get CinematicAvailable() -> (CinematicAvailable:bool)) -> (ReturnValue:float)
   )
   n4 = IsInFrontend!(self) -> (IsFrontend:bool)
   n5 = NOT Boolean(A=Get IsQuestScreen() -> (IsQuestScreen:bool))
   n6 = Get CinematicAvailable()
   > SetVisibility(
    self=Get Button_DebugQuestUI() -> (Button_DebugQuestUI:WBP UIKit Button Quiet Object Reference)
    InVisibility=Select(
     Option 0:ESlateVisibility Enum=Collapsed
     Option 1:ESlateVisibility Enum=Visible
     Index=AND Boolean!(A=@n6, B=@n4, C:bool=@n5) -> (ReturnValue)
    ) -> (ReturnValue:ESlateVisibility Enum)
   )
  ]}
 IsInFrontend {nodes [
   IsInFrontend() {is=title K2Node_FunctionEntry}
   > n1 = Cast To FortPlayerControllerFrontEnd!(execute, Object=GetOwningPlayer()) -> (then, CastFailed, AsFort Player Controller Front End:Fort Player Controller Front End Object Reference) {is=title K2Node_DynamicCast}
   > Return Node!(execute, IsFrontend:bool=true)
   Return Node(execute=@n1.CastFailed, IsFrontend=false)
  ]}
]
variables [
 CinematicAvailable {type=boolean}
 QuestUIRewardCinematicVM {type=QuestUIRewardCinematicVM, default=null}
 QuestVM {type=QuestUIBaseContentVM, default=null}
 VideoWidget {type=B_FortVideoWidgetBase_C, default=null}
 VUID {type=string, default=""}
]
widgets {
 info {
  inheritedWidgetCount=0
  parentClass=&/WBP_QuestRewardBase.WBP_QuestRewardBase_C
  rootWidgetClass=&SizeBox
  widgetCount=4
 }
 widgets [
  SizeBox_18 {widgetClassPath=&SizeBox, widget=SizeBox_18, properties {heightOverride=400, widthOverride=400}}
  Overlay_36 {widgetClassPath=&Overlay, parent=SizeBox_18, widget=Overlay_36, slot {class=SizeBoxSlot}}
  Button_DebugQuestUI {
   widgetClassPath=/FortUILibrary/UIKit/Buttons/WBP_UIKit_Button_Quiet
   bIsVariable=true
   parent=Overlay_36
   widget=Button_DebugQuestUI
   properties {
    bIsFocusable=false
    minWidth=120
    mobileMinWidth=200
    text=Play Cinematic
    visibility=Visible
    triggeringInputAction {dataTable=&InputActionDataTable, rowName=Confirm}
   }
   slot {class=OverlaySlot, properties {horizontalAlignment=HAlign_Center, verticalAlignment=VAlign_Bottom, padding=<ltrb 0 0 0 20>}}
  }
  PreviewImage {
   widgetClassPath=&Image
   bIsVariable=true
   parent=Overlay_36
   widget=PreviewImage
   properties {
    visibility=SelfHitTestInvisible
    brush {
     bIsDynamicallyLoaded=false
     drawAs=Image
     imageType=FullColor
     mirroring=NoMirror
     resourceName=None
     resourceObject=/QuestUI/QuestScreen/Materials/M_UI_CinematicRewardImage
     tiling=NoTile
     imageSize=<440 440>
     margin=<ltrb 0 0 0 0>
     outlineSettings {
      bUseBrushTransparency=false
      roundingType=HalfHeightRadius
      width=0
      color {colorUseRule=UseColor_Specified, specifiedColor=<rgba 0 0 0 0>}
      cornerRadii=<0 0 0 0>
     }
     tintColor {colorUseRule=UseColor_Specified, specifiedColor=<rgba 1 1 1 1>}
     uVRegion {bIsValid=false, max=<0 0>, min=<0 0>}
    }
   }
   slot {class=OverlaySlot, properties {horizontalAlignment=HAlign_Center, verticalAlignment=VAlign_Bottom, padding=<ltrb 0 0 0 20>}}
  }
 ]
}
