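# WBP_UIKit_QuestRewardDescription is a Widget Blueprint for quest reward descriptions, located in the quest screen's rewards panel.
# It derives from WBP_UIKit_ItemDescription_Base in the FortUILibrary item system.
# It lists no widgets or variables and has an empty event graph, so any layout or behavior comes from the parent class.
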WidgetBlueprint /QuestUI/QuestScreen/RewardsPanel/WBP_UIKit_QuestRewardDescription
graphs [EventGraph {links [], nodes []}]
variables []
widgets {
 info {
  inheritedWidgetCount=0
  parentClass=/FortUILibrary/Systems/Items/WBP_UIKit_ItemDescription_Base.WBP_UIKit_ItemDescription_Base_C
  rootWidgetClass=null
  widgetCount=0
 }
 widgets []
}
