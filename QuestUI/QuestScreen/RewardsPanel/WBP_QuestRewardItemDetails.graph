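# Quest UI widget that shows the details of a quest reward item, derived from WBP_QuestRewardBase.
# Its only widget is a WBP_UIKit_QuestRewardDescription as the root.
# An InitializeQuestRewardWidget override calls the parent, then passes the reward view model to that description widget through SetQuestUIRewardVM.
# The event graph is empty, so this asset adds no other logic.
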WidgetBlueprint /QuestUI/QuestScreen/RewardsPanel/WBP_QuestRewardItemDetails
graphs [
 EventGraph {links [], nodes []}
 InitializeQuestRewardWidget {nodes [
   n1 = InitializeQuestRewardWidget!() -> (
    then
    QuestContentVM:Quest UI Base Content View Model Object Reference
    QuestRewardVM:Quest UI Reward VM Object Reference
    IsQuestScreen:bool
   ) {is=title K2Node_FunctionEntry}
   > Parent: InitializeQuestRewardWidget!(
    execute
    QuestContentVM:Quest UI Base Content View Model Object Reference=@n1
    QuestRewardVM:Quest UI Reward VM Object Reference=@n1.QuestRewardVM
    IsQuestScreen:bool=@n1.IsQuestScreen
   ) -> (then) {is=title K2Node_CallParentFunction}
   > SetQuestUIRewardVM!(
    execute
    self:WBP UIKit Quest Reward Description Object Reference=Get WBP_UIKit_QuestRewardDescription() -> (WBP_UIKit_QuestRewardDescription:WBP UIKit Quest Reward Description Object Reference)
    Viewmodel:Quest UI Reward VM Object Reference=@n1.QuestRewardVM
   ) -> (then)
  ]}
]
variables []
widgets {
 info {
  inheritedWidgetCount=0
  parentClass=&/WBP_QuestRewardBase.WBP_QuestRewardBase_C
  rootWidgetClass=&/WBP_UIKit_QuestRewardDescription.WBP_UIKit_QuestRewardDescription_C
  widgetCount=1
 }
 widgets [WBP_UIKit_QuestRewardDescription {
   widgetClassPath=&/WBP_UIKit_QuestRewardDescription
   bIsVariable=true
   widget=WBP_UIKit_QuestRewardDescription
  }]
}
