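# Large quest reward item widget for the quest UI, derived from WBP_QuestRewardBase, showing a reward's preview image over a faint backplate.
# Initialization calls the parent, stores the reward view model and refreshes the icon, which also refreshes when the large preview image changes.
# The refresh lazily loads LargePreviewImage into the Reward image and shows the backplate only when the reward is cosmetic.
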WidgetBlueprint /QuestUI/QuestScreen/RewardsPanel/WBP_QuestRewardItemLarge
graphs [
 EventGraph {links [], nodes []}
 InitializeQuestRewardWidget {nodes [
   n1 = InitializeQuestRewardWidget!() -> (
    then
    QuestContentVM:Quest UI Base Content View Model Object Reference
    QuestRewardVM:Quest UI Reward VM Object Reference
    IsQuestScreen:bool
   ) {is=title K2Node_FunctionEntry}
   > Parent: InitializeQuestRewardWidget!(
    execute
    QuestContentVM:Quest UI Base Content View Model Object Reference=@n1
    QuestRewardVM:Quest UI Reward VM Object Reference=@n1.QuestRewardVM
    IsQuestScreen:bool=@n1.IsQuestScreen
   ) -> (then) {is=title K2Node_CallParentFunction}
   > Set QuestUIRewardVM(QuestUIRewardVM:Quest UI Reward VM Object Reference=@n1.QuestRewardVM) -> (Output_Get:Quest UI Reward VM Object Reference)
   > UpdateRewardIcon()
  ]}
 OnLargePreviewImageChanged {nodes [
   OnLargePreviewImageChanged!() -> (then, Icon:Soft Object Reference) {is=title K2Node_FunctionEntry}
   > UpdateRewardIcon()
  ]}
 UpdateRewardIcon {nodes [
   UpdateRewardIcon() {is=title K2Node_FunctionEntry}
   > SetBrushFromLazyDisplayAsset!(
    execute
    self:Common Lazy Image Object Reference=Get Reward() -> (Reward:Fort Lazy Image Object Reference)
    LazyObject:"Soft Object Reference (by ref)"=Get LargePreviewImage!(self:Quest UI Reward VM Object Reference=Get QuestUIRewardVM() -> (QuestUIRewardVM:Quest UI Reward VM Object Reference)) -> (LargePreviewImage:Soft Object Reference)
    bMatchTextureSize:bool=false
   ) -> (then)
   n1 = Get QuestUIRewardVM()
   > SetVisibility(
    self=Get Back() -> (Back:Image Object Reference)
    InVisibility=Select(Option 0=Collapsed, Option 1=HitTestInvisible, Index=Get bIsCosmetic!(self:Quest UI Reward VM Object Reference=@n1) -> (bIsCosmetic:bool))
   )
  ]}
]
variables [In Visibility {type=ESlateVisibility, default=HitTestInvisible}]
widgets {
 info {
  inheritedWidgetCount=0
  parentClass=&/WBP_QuestRewardBase.WBP_QuestRewardBase_C
  rootWidgetClass=&Overlay
  widgetCount=3
 }
 widgets [
  Overlay_0 {widgetClassPath=&Overlay, widget=Overlay_0}
  Back {
   widgetClassPath=&Image
   bIsVariable=true
   parent=Overlay_0
   widget=Back
   properties {
    brush {
     bIsDynamicallyLoaded=false
     drawAs=Image
     imageType=FullColor
     mirroring=NoMirror
     resourceName=None
     resourceObject=/FortUILibrary/Systems/Backplate/Materials/M_UIKit_BackplateCombined_Fallback
     tiling=NoTile
     imageSize=<256 256>
     margin=<ltrb 0 0 0 0>
     outlineSettings {
      bUseBrushTransparency=false
      roundingType=HalfHeightRadius
      width=0
      color {colorUseRule=UseColor_Specified, specifiedColor=<rgba 0 0 0 0>}
      cornerRadii=<0 0 0 0>
     }
     tintColor {colorUseRule=UseColor_Specified, specifiedColor=<rgba 1 1 1 1>}
     uVRegion {bIsValid=false, max=<0 0>, min=<0 0>}
    }
    colorAndOpacity=<rgba 1 1 1 0.1>
   }
   slot {class=OverlaySlot, properties {horizontalAlignment=HAlign_Center, verticalAlignment=VAlign_Center}}
  }
  Reward {
   widgetClassPath=&FortLazyImage
   bIsVariable=true
   parent=Overlay_0
   widget=Reward
   properties {
    visibility=Visible
    brush {
     bIsDynamicallyLoaded=false
     drawAs=Image
     imageType=NoImage
     mirroring=NoMirror
     resourceName=None
     resourceObject=null
     tiling=NoTile
     imageSize=<256 256>
     margin=<ltrb 0 0 0 0>
     outlineSettings {
      bUseBrushTransparency=false
      roundingType=HalfHeightRadius
      width=0
      color {colorUseRule=UseColor_Specified, specifiedColor=<rgba 0 0 0 0>}
      cornerRadii=<0 0 0 0>
     }
     tintColor {colorUseRule=UseColor_Specified, specifiedColor=<rgba 1 1 1 1>}
     uVRegion {bIsValid=false, max=<0 0>, min=<0 0>}
    }
   }
   slot {class=OverlaySlot, properties {horizontalAlignment=HAlign_Center}}
  }
 ]
}
