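# Season pass banner tag for the quest rewards panel, derived from WBP_UIKit_Tag, showing that a season item is required.
# SetSeasonItem collapses the tag when the item is invalid or disallowed, otherwise styles it pink with white "{0} Required" text using the item's display name.
# It then passes the item definition's pass data to LoadPassData, which async-loads that data and its SmallIcon material to set the tag icon.
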WidgetBlueprint /QuestUI/QuestScreen/RewardsPanel/WBP_SeasonPassBanner
graphs [
 EventGraph {nodes [
   n1 = LoadPassData!() -> (OutputDelegate, then, PassDataSoftObject:Season Pass Data Base Soft Object Reference) {is=title K2Node_CustomEvent}
   > n2 = Async Load Asset(Asset=@n1.PassDataSoftObject)
   n3 = Cast To SeasonPassData!(execute=@n2.Completed, Object=@n2) -> (then, CastFailed, AsSeason Pass Data:Season Pass Data Object Reference) {is=title K2Node_DynamicCast}
   > n4 = Async Load Asset(Asset=Get SmallIcon!(self:Season Pass Data Object Reference=@n3) -> (SmallIcon:Soft Object Reference))
   n5 = Cast To MaterialInstance!(execute=@n4.Completed, Object=@n4) -> (then, CastFailed, AsMaterial Instance:Material Instance Object Reference) {is=title K2Node_DynamicCast}
   > SetIconMaterial!(execute, self:WBP UIKit Tag Base Object Reference, IconMaterial:Material Instance Object Reference=@n5) -> (then)
   > SetIconColor!(execute, self:WBP UIKit Tag Base Object Reference, Color:Linear Color Structure=(R=1,G=1,B=1,A=1)) -> (then)
   > SetIconPadding!(execute, self:WBP UIKit Tag Base Object Reference, Padding:Margin Structure=Make Margin(Left=0, Top=0, Right=10, Bottom=0)) -> (then)
  ]}
 SetSeasonItem {nodes [
   n1 = SetSeasonItem!() -> (then, SeasonItem:Fort Season Item Object Reference, IsAllowed:bool=~) {is=title K2Node_FunctionEntry}
   > n2 = Is Valid(InputObject=@n1)
   > n3 = Branch(Condition=@n1.IsAllowed)
   > SetVisibility(InVisibility=HitTestInvisible)
   > n4 = Sequence()
   n5 = Get SeasonItem() -> (SeasonItem:Fort Season Item Object Reference)
   > SetTagText!(
    execute
    self:WBP UIKit Tag Base Object Reference
    Text:Text=Format Text!(Format="LOCTEXT(\"{0} Required\")", 0:Text=GetDisplayName!(self:Fort Item Object Reference=@n5) -> (ReturnValue:Text)) -> (Result)
   ) -> (then)
   > SetTagTextColor!(execute, self:WBP UIKit Tag Base Object Reference, Color:Linear Color Structure=(R=1,G=1,B=1,A=1)) -> (then)
   > SetTagBackgroundColor!(execute, self:WBP UIKit Tag Base Object Reference, Color:Linear Color Structure=(R=0.807,G=0.0203,B=0.3419,A=1)) -> (then)
   n6 = Cast To SeasonPassItemDefinition!(execute=@n4.then_1, Object=GetItemDefinition!(self:Fort Item Object Reference=@n5) -> (ReturnValue:Fort Item Definition Object Reference)) -> (then, CastFailed, AsSeason Pass Item Definition:Season Pass Item Definition Object Reference) {is=title K2Node_DynamicCast}
   > LoadPassData!(
    execute
    self
    PassDataSoftObject:Season Pass Data Base Soft Object Reference=Get PassData!(self:Season Pass Item Definition Object Reference=@n6) -> (PassData:Season Pass Data Base Soft Object Reference)
   ) -> (then) {is=title K2Node_CallFunction}
   SetVisibility(execute=[@n3.else, @n2.Is Not Valid], InVisibility=Collapsed)
  ]}
]
variables []
widgets {
 info {inheritedWidgetCount=0, parentClass=/FortUILibrary/UIKit/Tag/WBP_UIKit_Tag.WBP_UIKit_Tag_C, rootWidgetClass=null, widgetCount=0}
 widgets []
}
