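# Side panel widget of the quest UI that previews a quest's large-eligible rewards, one page at a time.
# UpdateRewards filters reward view models and sizes the pager, whose selection drives SetActiveReward.
# That event async-creates preview and details widgets from class maps keyed by view model type, discarding stale previews, and updates the pass-required banner.
# Mobile gets compact layout and optional hiding, while design time generates random debug rewards.
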WidgetBlueprint /QuestUI/QuestScreen/RewardsPanel/WBP_SideRewardsPanel
graphs [
 DEBUG_RandomGameModes {nodes [
   DEBUG_RandomGameModes() {is=title K2Node_FunctionEntry}
   > Clear(TargetArray:Array of E_GameModes Enums=Get DEBUG_CurrentAvaliableModes() -> (DEBUG_CurrentAvaliableModes:Array of E_GameModes Enums=NewEnumerator0))
   n1 = Get DEBUG_Avaliablemodes() -> (DEBUG_Avaliablemodes:Array of E_GameModes Enums=NewEnumerator0)
   > n2 = For Loop(FirstIndex=1, LastIndex=RandomInteger!(Max:int=Length(TargetArray:Array of E_GameModes Enums=@n1)) -> (ReturnValue:int=0))
   > Add(
    TargetArray:Array of E_GameModes Enums=Get DEBUG_CurrentAvaliableModes() -> (DEBUG_CurrentAvaliableModes=NewEnumerator0)
    NewItem:"E_GameModes Enum (by ref)"="Get (a copy)"(Array:Array of E_GameModes Enums=Get DEBUG_Avaliablemodes() -> (DEBUG_Avaliablemodes=NewEnumerator0), Dimension 1=@n2) -> (Output:E_GameModes Enum)
   ) {is=title K2Node_CallArrayFunction}
  ]}
 EventGraph {nodes [
   n1 = "Selected Page Changed (Pages)"!() -> (OutputDelegate, then, SelectedPip:int=0) {is=title K2Node_ComponentBoundEvent}
   > SetActiveReward!(execute, self, Index:int=@n1.SelectedPip) -> (then)
   n2 = SetActiveReward!() -> (OutputDelegate, then, Index:int) {is=title K2Node_CustomEvent}
   > Set ActiveRewardIndex(ActiveRewardIndex:int=@n2.Index) -> (Output_Get:int=0)
   > Set QuestUIRewardVM(QuestUIRewardVM:Quest UI Reward VM Object Reference) -> (Output_Get:Quest UI Reward VM Object Reference)
   > n3 = Branch(Condition=Is Valid Index!(
     TargetArray:Array of Quest UI Reward VM Object References=Get CustomWidgetRewards() -> (CustomWidgetRewards:Array of Quest UI Reward VM Object References)
     IndexToTest:int=Get ActiveRewardIndex() -> (ActiveRewardIndex:int=0)
    ) -> (ReturnValue:bool))
   > Set QuestUIRewardVM(QuestUIRewardVM="Get (a copy)"(Array:Array of Quest UI Reward VM Object References=Get CustomWidgetRewards(), Dimension 1=Get ActiveRewardIndex() -> (ActiveRewardIndex=0)) -> (Output:Quest UI Reward VM Object Reference))
   > ClearChildren(self=Get CustomWidgetNamedSlot() -> (CustomWidgetNamedSlot:Named Slot Object Reference))
   n4 = Get QuestUIRewardVM() -> (QuestUIRewardVM:Quest UI Reward VM Object Reference)
   > n5 = Set ActivePreviewWidget(ActivePreviewWidget:User Widget Soft Class Reference=Find(
     TargetMap:Map of Quest UI Reward VM Class References to User Widget Soft Class References=Get PreviewWidgetsByReward() -> (PreviewWidgetsByReward:Map of Quest UI Reward VM Class References to User Widget Soft Class References)
     Key:"Quest UI Reward VM Class Reference (by ref)"=Get Class(Object=@n4) -> (ReturnValue:Quest UI Reward VM Class Reference)
    ) -> (Value:User Widget Soft Class Reference)) -> (Output_Get:User Widget Soft Class Reference)
   > n6 = CreateWidgetAsync!(
    execute:Exec
    UserWidgetSoftClass:User Widget Soft Class Reference=@n5
    OwningPlayer:Player Controller Object Reference=GetOwningPlayer()
    bSuspendInputUntilComplete:bool=false
   ) -> (then:Exec, OnComplete:Exec, OnFailure:Exec, UserWidget:User Widget Object Reference) {is=title K2Node_AsyncAction}
   Branch(
    execute=@n6.OnComplete
    Condition="Equal (Class)"(
     A=Get Class(Object=@n6) -> (ReturnValue:User Widget Class Reference)
     B:User Widget Soft Class Reference=Get ActivePreviewWidget() -> (ActivePreviewWidget:User Widget Soft Class Reference)
    )
   )
   > n7 = Cast To WBP_QuestRewardBase!(execute, Object=@n6) -> (then, CastFailed, AsWBP Quest Reward Base:WBP Quest Reward Base Object Reference) {is=title K2Node_DynamicCast}
   > InitializeQuestRewardWidget!(
    execute
    self:WBP Quest Reward Base Object Reference=@n7
    QuestContentVM:Quest UI Base Content View Model Object Reference=Get Quest VM() -> (Quest VM:Quest UI Base Content View Model Object Reference)
    QuestRewardVM:Quest UI Reward VM Object Reference=Get QuestUIRewardVM()
    IsQuestScreen:bool=Get IsQuestScreen() -> (IsQuestScreen:bool)
   ) -> (then)
   > AddChild(self=Get CustomWidgetNamedSlot(), Content=@n6)
   > UpdatePassBanner()
   > ClearChildren(self=Get DetailsWidgetSlot() -> (DetailsWidgetSlot:Overlay Object Reference))
   n8 = Get QuestUIRewardVM()
   > n9 = CreateWidgetAsync(
    UserWidgetSoftClass=Find(
     TargetMap=Get DetailsWidgetsByReward() -> (DetailsWidgetsByReward:Map of Quest UI Reward VM Class References to User Widget Soft Class References)
     Key=Get Class(Object=@n8) -> (ReturnValue:Quest UI Reward VM Class Reference)
    )
    OwningPlayer=GetOwningPlayer()
    bSuspendInputUntilComplete=false
   )
   n10 = Cast To WBP_QuestRewardBase(execute=@n9.OnComplete, Object=@n9)
   > InitializeQuestRewardWidget(self=@n10, QuestContentVM=Get Quest VM(), QuestRewardVM=Get QuestUIRewardVM(), IsQuestScreen=Get IsQuestScreen())
   > n11 = AddChildToOverlay!(execute, self:Overlay Object Reference=Get DetailsWidgetSlot(), Content:Widget Object Reference=@n9) -> (then, ReturnValue:Overlay Slot Object Reference)
   > SetHorizontalAlignment(self=@n11, InHorizontalAlignment=HAlign_Fill)
   Set ActiveRewardIndex(execute=@n3.else, ActiveRewardIndex=-1) -> (Output_Get=0)
   n12 = Event PreConstruct()
   > SetCompactUIStyle() {is=title K2Node_CallFunction}
   > n13 = Branch(Condition=@n12.IsDesignTime)
   > n14 = For Loop(LastIndex=10)
   SetOptionsAmount!(
    execute=@n14.Completed
    self
    Amount:int=Length(TargetArray:Array of MVVM Debug Quests Reward Object References=Get Rewards() -> (Rewards:Array of MVVM Debug Quests Reward Object References))
    StartingPip:int=0
   ) -> (then)
   n15 = Construct Object from Class!(
    execute=@n14.LoopBody
    Class=&/MVVM_DebugQuestsReward.MVVM_DebugQuestsReward_C
    self
    RewardType:E_RewardTypes Enum=Random Array Item!(TargetArray:Array of E_RewardTypes Enums=Get DEBUG_RewardsTypes() -> (DEBUG_RewardsTypes:Array of E_RewardTypes Enums=NewEnumerator3)) -> (OutItem:E_RewardTypes Enum=NewEnumerator3, OutIndex:int=0) {is=title K2Node_CallArrayFunction}
    RewardName:Text=Random Array Item(TargetArray:Array of Texts=Get DEBUG_RewardsName() -> (DEBUG_RewardsName:Array of Texts)) -> (OutItem:Text, OutIndex=0)
    AvaliableModes:Array of E_GameModes Enums=Get DEBUG_Avaliablemodes() -> (DEBUG_Avaliablemodes=NewEnumerator0)
    AlternativeUsableText:Text=Random Array Item(TargetArray=Get DEBUG_AlternativeUsableText() -> (DEBUG_AlternativeUsableText:Array of Texts)) -> (OutIndex=0)
    RewardTexture:Texture 2D Object Reference=Random Array Item(TargetArray:Array of Texture 2D Object References=Get DEBUG_Reward Texture() -> (DEBUG_Reward Texture:Array of Texture 2D Object References)) -> (OutItem:Texture 2D Object Reference, OutIndex=0)
   ) -> (then, ReturnValue:MVVM Debug Quests Reward Object Reference)
   > Add(
    TargetArray:Array of MVVM Debug Quests Reward Object References=Get Rewards()
    NewItem:"MVVM Debug Quests Reward Object Reference (by ref)"=@n15
   )
   Set AlwaysShowHints!(
    execute=@n13.else
    AlwaysShowHints:bool=Get IsQuestScreen()
    self:WBP Quest Pages Object Reference=Get Pages() -> (Pages:WBP Quest Pages Object Reference)
   ) -> (then, Output_Get:bool)
   Event Construct()
  ]}
 SetCompactUIStyle {nodes [
   SetCompactUIStyle() {is=title K2Node_FunctionEntry}
   > Branch(Condition=IsMobileGame())
   > SetPadding!(
    execute
    self:Overlay Slot Object Reference=SlotAsOverlaySlot()
    InPadding_Left:float=0
    InPadding_Top:float=520
    InPadding_Right:float=0
    InPadding_Bottom:float=0
   ) -> (then)
   n1 = Get SizeBox_Root() -> (SizeBox_Root:Size Box Object Reference)
   > SetHeightOverride(self=@n1, InHeightOverride=700)
   > SetWidthOverride(self=@n1, InWidthOverride=700)
   > SetPadding(self=SlotAsOverlaySlot(Widget=Get Pages()), InPadding_Left=0, InPadding_Top=50, InPadding_Right=0, InPadding_Bottom=0)
  ]}
 SetOptionsAmount {nodes [
   SetOptionsAmount!() -> (then, Amount:int, StartingPip:int) {is=title K2Node_FunctionEntry}
   n1 = Get Amount() -> (Amount:int=0)
   > SetVisibility(self=Get Pages(), InVisibility=Select(Option 0=Collapsed, Option 1=SelfHitTestInvisible, Index="integer > integer"(A=@n1, B=1)))
   > SetOptionsAmount!(execute, self:WBP Quest Pages Object Reference=Get Pages(), PagesAmount:int=Get Amount() -> (Amount=0)) -> (then) {is=title K2Node_CallFunction}
   > SetSelectedPage!(execute, self:WBP Quest Pages Object Reference=Get Pages(), SelectedOption:int=Get StartingPip() -> (StartingPip:int=0)) -> (then)
  ]}
 TrySetDesiredVisibility {nodes [
   TrySetDesiredVisibility!() -> (then, DesiredVisibility:ESlateVisibility Enum=Visible) {is=title K2Node_FunctionEntry}
   n1 = Get bCanBeVisibleOnMobile() -> (bCanBeVisibleOnMobile:bool)
   > n2 = Branch(Condition=AND Boolean(A=IsMobileGame(), B=NOT Boolean(A=@n1)))
   > SetVisibility(InVisibility=Collapsed)
   SetVisibility(execute=@n2.else, InVisibility=Get DesiredVisibility() -> (DesiredVisibility:ESlateVisibility Enum=Visible))
  ]}
 UpdatePassBanner {nodes [
   UpdatePassBanner() {is=title K2Node_FunctionEntry}
   n1 = Get QuestUIRewardVM()
   n2 = NOT Boolean(A=Get bIsRewardContentAvailable!(self:Quest UI Reward VM Object Reference=@n1) -> (bIsRewardContentAvailable:bool))
   > SetSeasonItem!(
    execute
    self:WBP Season Pass Banner Object Reference=Get PassBanner() -> (PassBanner:WBP Season Pass Banner Object Reference)
    SeasonItem:Fort Season Item Object Reference=GetRequiredSeasonItem!(self:Quest UI Reward VM Object Reference=@n1) -> (ReturnValue:Fort Season Item Object Reference)
    IsAllowed:bool=@n2
   ) -> (then)
   > n3 = Branch(Condition=@n2)
   n4 = Get BannerSwitch() -> (BannerSwitch:Common Visibility Switcher Object Reference)
   > SetVisibility(self=@n4, InVisibility=HitTestInvisible)
   n5 = GetRequiredSeasonItem(self=Get QuestUIRewardVM())
   > Set Text(
    Text=Format Text!(Format="LOCTEXT(\"{0} Required\")", 0:Text=GetDisplayName!(self:Fort Item Object Reference=@n5) -> (ReturnValue:Text)) -> (Result)
    self=Get Text_PassRequired() -> (Text_PassRequired:Common Text Object Reference)
   )
   SetVisibility(execute=@n3.else, self=@n4, InVisibility=Collapsed)
  ]}
 UpdateRewards {nodes [
   n1 = UpdateRewards!() -> (then, RewardsCustomWidgets:Array of Quest UI Reward VM Object References, QuestContentVM:Quest UI Base Content View Model Object Reference) {is=title K2Node_FunctionEntry}
   > Set Quest VM(Quest VM:Quest UI Base Content View Model Object Reference=@n1.QuestContentVM) -> (Output_Get:Quest UI Base Content View Model Object Reference)
   > Clear(TargetArray:Array of Quest UI Reward VM Object References=Get CustomWidgetRewards())
   > n2 = For Each Loop(Array:Array of Quest UI Reward VM Object References=@n1) -> (Array Element:Quest UI Reward VM Object Reference)
   n3 = Branch(
    execute=@n2.Completed
    Condition=Is Empty!(TargetArray:Array of Quest UI Reward VM Object References=Get CustomWidgetRewards()) -> (ReturnValue:bool)
   )
   > TrySetDesiredVisibility!(execute, self, DesiredVisibility:ESlateVisibility Enum=Collapsed) -> (then) {is=title K2Node_CallFunction}
   > SetOptionsAmount!(execute, self:Self Object Reference, Amount:int=0, StartingPip:int=0) -> (then)
   TrySetDesiredVisibility(execute=@n3.else, DesiredVisibility=SelfHitTestInvisible)
   > SetOptionsAmount(Amount=Length(TargetArray:Array of Quest UI Reward VM Object References=Get CustomWidgetRewards()), StartingPip=0)
   Branch(execute=@n2.LoopBody, Condition=IsEligibleForShowingInLargeReward!(self:Quest UI Reward VM Object Reference=@n2) -> (ReturnValue:bool))
   > Add(
    TargetArray:Array of Quest UI Reward VM Object References=Get CustomWidgetRewards()
    NewItem:"Quest UI Reward VM Object Reference (by ref)"=@n2
   )
  ]}
]
variables [
 ActivePreviewWidget {type=Class, default=null}
 ActiveRewardIndex {type=integer}
 bCanBeVisibleOnMobile {type=boolean, default=true}
 CustomWidgetRewards {type="array<QuestUIRewardVM>", default []}
 DEBUG_AlternativeUsableText {type="array<string>", default ["", Used to increase the Fortnite Season Level]}
 DEBUG_Avaliablemodes {type="array<E_GameModes>", default [Battle Royale, Fortnite Festival, Rocket Racing, "Lego® Fortnite"]}
 DEBUG_CurrentAvaliableModes {type="array<E_GameModes>", default [Battle Royale, Fortnite Festival, Rocket Racing, "Lego® Fortnite"]}
 DEBUG_Reward Texture {
  type="array<Texture2D>"
  default [
   /Game/UI/Foundation/Textures/Icons/Items/UniversalXP/T_UI_UniversalXP_Icon_Standard_128
   /Game/UI/SurvivorTiers/Textures/S-Tiers-New_T_UI_SurvivalTierIcon_Tier3
  ]
 }
 DEBUG_RewardsName {type="array<string>", default [PRIZE LLAMA, "100,000 xp", outfit name goes here...]}
 DEBUG_RewardsTypes {type="array<E_RewardTypes>", default [None, BACK BLING, OUTFIT, WRAP]}
 DetailsWidgetsByReward {default {
   &Class'/Script/QuestUI.QuestUIRewardCinematicVM'=&/WBP_QuestRewardCinematicDetails.WBP_QuestRewardCinematicDetails_C
   &Class'/Script/QuestUI.QuestUIRewardVM'=&/WBP_QuestRewardItemDetails.WBP_QuestRewardItemDetails_C
  }}
 GameTagClass {type=Class, default=/FortUILibrary/UIKit/Tag/WBP_UIKit_Tag.WBP_UIKit_Tag_C}
 GameTagType {type=E_UI_TagType, default=Secondary Tag}
 IsQuestScreen {type=boolean}
 PreviewWidgetsByReward {default {
   &Class'&QuestUIRewardCinematicVM'=&/WBP_QuestRewardCinematicLarge.WBP_QuestRewardCinematicLarge_C
   &Class'&QuestUIRewardVM'=&/WBP_QuestRewardItemLarge.WBP_QuestRewardItemLarge_C
  }}
 ProductTagsName {type="array<string>", default []}
 Quest VM {type=QuestUIBaseContentVM, default=null}
 Rewards {type="array<MVVM_DebugQuestsReward_C>", default []}
]
widgets {
 info {inheritedWidgetCount=0, parentClass=&CommonUserWidget, rootWidgetClass=&SizeBox, widgetCount=16}
 widgets [
  SizeBox_Root {
   widgetClassPath=&SizeBox
   bIsVariable=true
   widget=SizeBox_Root
   properties {
    bOverride_MaxDesiredHeight=true
    bOverride_MinDesiredHeight=true
    bOverride_WidthOverride=true
    heightOverride=700
    maxDesiredHeight=700
    minDesiredHeight=650
    widthOverride=650
   }
  }
  Overlay_0 {widgetClassPath=&Overlay, parent=SizeBox_Root, widget=Overlay_0, slot {class=SizeBoxSlot}}
  FortVisualAttachment_Glow {
   widgetClassPath=&FortVisualAttachment
   bIsVariable=true
   parent=Overlay_0
   widget=FortVisualAttachment_Glow
   properties {contentAnchor=<0.5 0.45>}
   slot {class=OverlaySlot, properties {horizontalAlignment=HAlign_Center, verticalAlignment=VAlign_Fill, padding=<ltrb 0 0 0 416>}}
  }
  RewardGlow {
   widgetClassPath=&Image
   bIsVariable=true
   parent=FortVisualAttachment_Glow
   widget=RewardGlow
   properties {
    visibility=SelfHitTestInvisible
    brush {
     bIsDynamicallyLoaded=false
     drawAs=Image
     imageType=FullColor
     mirroring=NoMirror
     resourceName=None
     resourceObject=/QuestUI/QuestScreen/MaterialInstances/MI_QuestRewardRadialGlow
     tiling=NoTile
     imageSize=<1859.5725 927.7485>
     margin=<ltrb 0 0 0 0>
     outlineSettings {
      bUseBrushTransparency=false
      roundingType=HalfHeightRadius
      width=0
      color {colorUseRule=UseColor_Specified, specifiedColor=<rgba 0 0 0 0>}
      cornerRadii=<0 0 0 0>
     }
     tintColor {colorUseRule=UseColor_Specified, specifiedColor=<rgba 1 1 1 1>}
     uVRegion {bIsValid=false, max=<0 0>, min=<0 0>}
    }
    renderTransform {angle=-35, scale=<1 1>, shear=<0 0>, translation=<0 0>}
   }
   slot {class=SizeBoxSlot}
  }
  VerticalBox_0 {
   widgetClassPath=&VerticalBox
   parent=Overlay_0
   widget=VerticalBox_0
   slot {class=OverlaySlot, properties {horizontalAlignment=HAlign_Center, verticalAlignment=VAlign_Bottom}}
  }
  CustomWidgetNamedSlot {widgetClassPath=&NamedSlot, bIsVariable=true, parent=VerticalBox_0, widget=CustomWidgetNamedSlot, slot {class=VerticalBoxSlot}}
  Pages {
   widgetClassPath=&/WBP_QuestPages
   bIsVariable=true
   parent=VerticalBox_0
   widget=Pages
   properties {pagesAmount=8}
   slot {class=VerticalBoxSlot, properties {horizontalAlignment=HAlign_Center, verticalAlignment=VAlign_Center, padding=<ltrb 0 30 0 20>}}
  }
  BannerSwitch {
   widgetClassPath=&CommonVisibilitySwitcher
   bIsVariable=true
   parent=VerticalBox_0
   widget=BannerSwitch
   properties {activeWidgetIndex=1}
   slot {class=VerticalBoxSlot, properties {padding=<ltrb 0 0 0 4>}}
  }
  PassBanner {
   widgetClassPath=&/WBP_SeasonPassBanner
   bIsVariable=true
   parent=BannerSwitch
   widget=PassBanner
   slot {class=CommonVisibilitySwitcherSlot}
  }
  Tag_PassRequired {
   widgetClassPath=&Border
   bIsVariable=true
   parent=BannerSwitch
   widget=Tag_PassRequired
   properties {
    background {
     bIsDynamicallyLoaded=false
     drawAs=Image
     imageType=FullColor
     mirroring=NoMirror
     resourceName=None
     resourceObject=/FortUILibrary/Systems/Badge/Materials/M_UIKit_Badge_Background
     tiling=NoTile
     imageSize=<32 32>
     margin=<ltrb 0 0 0 0>
     outlineSettings {
      bUseBrushTransparency=false
      roundingType=HalfHeightRadius
      width=0
      color {colorUseRule=UseColor_Specified, specifiedColor=<rgba 0 0 0 0>}
      cornerRadii=<0 0 0 0>
     }
     tintColor {colorUseRule=UseColor_Specified, specifiedColor=<rgba 1 0.063 0.3372 1>}
     uVRegion {bIsValid=false, max=<0 0>, min=<0 0>}
    }
    padding=<ltrb 8 2 8 2>
   }
   slot {class=CommonVisibilitySwitcherSlot, properties {horizontalAlignment=HAlign_Left, verticalAlignment=VAlign_Bottom}}
  }
  HorizontalBox_0 {
   widgetClassPath=&HorizontalBox
   parent=Tag_PassRequired
   widget=HorizontalBox_0
   slot {class=BorderSlot, properties {padding=<ltrb 8 2 8 2>}}
  }
  Image_BPIcon {
   widgetClassPath=&Image
   bIsVariable=true
   parent=HorizontalBox_0
   widget=Image_BPIcon
   properties {brush {
     bIsDynamicallyLoaded=false
     drawAs=Image
     imageType=FullColor
     mirroring=NoMirror
     resourceName=None
     resourceObject=/QuestUI/QuestScreen/Textures/Icon_Battle_Pass_Quest
     tiling=NoTile
     imageSize=<32 32>
     margin=<ltrb 0 0 0 0>
     outlineSettings {
      bUseBrushTransparency=false
      roundingType=HalfHeightRadius
      width=0
      color {colorUseRule=UseColor_Specified, specifiedColor=<rgba 0 0 0 0>}
      cornerRadii=<0 0 0 0>
     }
     tintColor {colorUseRule=UseColor_Specified, specifiedColor=<rgba 1 1 1 1>}
     uVRegion {bIsValid=false, max=<0 0>, min=<0 0>}
    }}
   slot {class=HorizontalBoxSlot, properties {padding=<ltrb 5 0 5 0>}}
  }
  Text_PassRequired {
   widgetClassPath=&CommonTextBlock
   bIsVariable=true
   parent=HorizontalBox_0
   widget=Text_PassRequired
   properties {
    justification=Center
    mobileFontSizeMultiplier=1.14
    scrollStyle=&TextScrollStyle-Base_C
    text=BATTLE PASS REQUIRED
    textTransformPolicy=ToUpper
    font {
     bForceMonospaced=false
     bMaterialIsStencil=false
     fontMaterial=null
     fontObject=&HeadingNow
     letterSpacing=0
     monospacedWidth=1
     size=15.75
     skewAmount=0
     typefaceFontName=86 Bold Italic
     outlineSettings {
      bApplyOutlineToDropShadows=false
      bMiteredCorners=false
      bSeparateFillAlpha=false
      outlineMaterial=null
      outlineSize=0
      outlineColor=<rgba 0 0 0 1>
     }
    }
    shadowOffset=<0 0>
   }
   slot {class=HorizontalBoxSlot, properties {verticalAlignment=VAlign_Center, padding=<ltrb 0 0 5 0>}}
  }
  SizeBox_1 {
   widgetClassPath=&SizeBox
   parent=VerticalBox_0
   widget=SizeBox_1
   properties {bOverride_MinDesiredHeight=true, bOverride_MinDesiredWidth=true, maxDesiredHeight=250, minDesiredHeight=250, minDesiredWidth=700}
   slot {class=VerticalBoxSlot, properties {horizontalAlignment=HAlign_Left, verticalAlignment=VAlign_Top}}
  }
  DetailsWidgetSlot {widgetClassPath=&Overlay, bIsVariable=true, parent=SizeBox_1, widget=DetailsWidgetSlot, slot {class=SizeBoxSlot}}
  WBP_UIKit_QuestRewardDescription {
   widgetClassPath=&/WBP_UIKit_QuestRewardDescription
   bIsVariable=true
   parent=DetailsWidgetSlot
   widget=WBP_UIKit_QuestRewardDescription
   slot {class=OverlaySlot}
  }
 ]
}
